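; Pawn Blueprint for Robert, an Irwin predator NPC, that overrides interaction and auto-fire targeting rules.
; Interaction requires the parent check to pass and no blocking tags on the interacting pawn or, unless creative taming tags override them, on itself.
; It is also refused when a different converter pawn exists and the NPC is at either of two alert levels.
; Auto-fire accepts it only for hostile attackers; events and construction add no logic.
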(event EventBeginPlay
  (|CallParentFunction) ; This node is disabled and will not be called.\nDrag off pins to build functionality.
)

(event Collision|EventActorBeginOverlap
  (params (OtherActor :type Actor%20Object%20Reference))
  (|CallParentFunction OtherActor) ; This node is disabled and will not be called.\nDrag off pins to build functionality.
)

(event EventTick
  (params (DeltaSeconds :type Float%20%28single-precision%29))
  (|CallParentFunction DeltaSeconds) ; This node is disabled and will not be called.\nDrag off pins to build functionality.
)

(fn BlueprintCanInteract
  (params (InteractingPawn :type Fort%20Pawn%20Object%20Reference) (InteractionBeingAttempted :type EInteractionBeingAttempted%20Enum :default "FirstInteraction"))
  (return (Math|Boolean|ANDBoolean (|CallParentFunction InteractingPawn InteractionBeingAttempted) (not (GameplayTags|HasAnyMatchingGameplayTags InteractingPawn (Variables|Default|GetInteractingPawnPreventingTags))) (select (GameplayTags|HasAnyMatchingGameplayTags (LocalVariables|GetCreative%20Taming%20Allowed%20Tags)) true (not (GameplayTags|HasAnyMatchingGameplayTags (Variables|Default|GetInteraction%20Preventing%20Tags)))) (not (Math|Boolean|ANDBoolean (Utilities|IsValid (Convert|GetConverterPawn (Variables|Default|GetConvertComponent)) :!class /Script/Engine.KismetSystemLibrary) (!= (Convert|GetConverterPawn (Variables|Default|GetConvertComponent)) (Variables|GetInteractingPawn)) (or (Utilities|Enum|Equal(Enum) (Variables|NPCAlertLevel|GetNPC_AlertLevel) "NewEnumerator2") (Utilities|Enum|Equal(Enum) (Variables|NPCAlertLevel|GetNPC_AlertLevel) "NewEnumerator3")))))))

(fn IsValidAutoFireTarget
  (params (Attacker :type Actor%20Object%20Reference))
  (return (Utilities|Enum|Equal(Enum) (Team|GetTeamAffiliation Attacker self) "Hostile")))

(fn ConstructionScript
  (|CallParentFunction))
